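# Physics asset for the RubySnapper pickaxe mesh, defining secondary motion for a figure-like prop with arm, hand, head, hair and wires.
# Small capsule bodies cover the dynamic bones, with larger capsules on the hand and Melee_TwoHanded_Handle.
# Constraint chains run from the handle through arm, wrist, hand and jointed fingers, through two body wires, and through neck and head to hair and head wires.
# Narrow swing and twist limits restrain movement, with finger joints bending mainly on one axis.
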PhysicsAsset /BRCosmetics/Weapons/FORT_Melee/Pickaxe_RubySnapper/Meshes/RubySnapper_Axe_Physics
bodies {
 Melee_TwoHanded_Handle [{shapeType=Capsule, length=8.5, radius=2.5, shapeName=Melee_TwoHanded_Handle_capsule, center=<0.142 -1.3012 42.4361>}]
 dyn_arm_00 [{shapeType=Capsule, length=2, radius=1, shapeName=dyn_arm_00_capsule, rotation=<rot 90 0 0>}]
 dyn_bodyWireA_00 [{shapeType=Capsule, length=2, radius=1, shapeName=dyn_bodyWireA_00_capsule, rotation=<rot 90 0 0>}]
 dyn_bodyWireA_01 [{shapeType=Capsule, length=2, radius=1, shapeName=dyn_bodyWireA_01_capsule, rotation=<rot 90 0 0>}]
 dyn_bodyWireA_02 [{shapeType=Capsule, length=2, radius=1, shapeName=dyn_bodyWireA_02_capsule, rotation=<rot 90 0 0>}]
 dyn_bodyWireB_00 [{shapeType=Capsule, length=2, radius=1, shapeName=dyn_bodyWireB_00_capsule, rotation=<rot 90 0 0>}]
 dyn_bodyWireB_01 [{shapeType=Capsule, length=2, radius=1, shapeName=dyn_bodyWireB_01_capsule, rotation=<rot 90 0 0>}]
 dyn_bodyWireB_02 [{shapeType=Capsule, length=2, radius=1, shapeName=dyn_bodyWireB_02_capsule, rotation=<rot 90 0 0>}]
 dyn_hairA_00 [{shapeType=Capsule, length=2, radius=1, shapeName=dyn_hairA_00_capsule, rotation=<rot 90 0 0>}]
 dyn_hairB_00 [{shapeType=Capsule, length=2, radius=1, shapeName=dyn_hairB_00_capsule, rotation=<rot 90 0 0>}]
 dyn_hairC_00 [{shapeType=Capsule, length=2, radius=1, shapeName=dyn_hairC_00_capsule, rotation=<rot 90 0 0>}]
 dyn_hairD_00 [{shapeType=Capsule, length=2, radius=1, shapeName=dyn_hairD_00_capsule, rotation=<rot 90 0 0>}]
 dyn_hairE_00 [{shapeType=Capsule, length=2, radius=1, shapeName=dyn_hairE_00_capsule, rotation=<rot 90 0 0>}]
 dyn_hand_00 [{
   shapeType=Capsule
   length=1
   radius=3.5
   shapeName=dyn_hand_00_capsule
   center=<-0.7028 -0.0076 0.0113>
   rotation=<rot -6.8791 175.3906 -142.1172>
  }]
 dyn_head [{shapeType=Capsule, length=2, radius=1, shapeName=dyn_head_capsule, center=<2 0 0>, rotation=<rot 90 0 0>}]
 dyn_headWireA_00 [{shapeType=Capsule, length=2, radius=1, shapeName=dyn_headWireA_00_capsule, rotation=<rot 90 0 0>}]
 dyn_headWireB_00 [{shapeType=Capsule, length=2, radius=1, shapeName=dyn_headWireB_00_capsule, rotation=<rot 90 0 0>}]
 dyn_headWireC_00 [{shapeType=Capsule, length=2, radius=1, shapeName=dyn_headWireC_00_capsule, rotation=<rot 90 0 0>}]
 dyn_headWireD_00 [{shapeType=Capsule, length=2, radius=1, shapeName=dyn_headWireD_00_capsule, rotation=<rot 90 0 0>}]
 dyn_index_00 [{shapeType=Capsule, length=1, radius=1, shapeName=dyn_index_00_capsule, center=<1 0 0>, rotation=<rot 90 0 0>}]
 dyn_index_02 [{shapeType=Capsule, length=1, radius=1, shapeName=dyn_index_02_capsule, center=<1 0 0>, rotation=<rot 90 0 0>}]
 dyn_middle_00 [{shapeType=Capsule, length=1, radius=0.5, shapeName=dyn_middle_00_capsule, center=<1 0 0>, rotation=<rot 90 0 0>}]
 dyn_middle_02 [{shapeType=Capsule, length=1, radius=1, shapeName=dyn_middle_02_capsule, center=<1 0 0>, rotation=<rot 90 0 0>}]
 dyn_neck [{shapeType=Capsule, length=2, radius=1, shapeName=dyn_neck_capsule, center=<2 0 0>, rotation=<rot 90 0 0>}]
 dyn_pinky_00 [{shapeType=Capsule, length=1, radius=0.5, shapeName=dyn_pinky_00_capsule, center=<1 0 0>, rotation=<rot 90 0 0>}]
 dyn_pinky_02 [{shapeType=Capsule, length=1, radius=1, shapeName=dyn_pinky_02_capsule, center=<1 0 0>, rotation=<rot 90 0 0>}]
 dyn_ring_00 [{shapeType=Capsule, length=1, radius=0.5, shapeName=dyn_ring_00_capsule, center=<1 0 0>, rotation=<rot 90 0 0>}]
 dyn_ring_02 [{shapeType=Capsule, length=1, radius=1, shapeName=dyn_ring_02_capsule, center=<1 0 0>, rotation=<rot 90 0 0>}]
 dyn_thumb_00 [{shapeType=Capsule, length=1, radius=1, shapeName=dyn_thumb_00_capsule, center=<1 0 0>, rotation=<rot 90 0 0>}]
 dyn_thumb_01 [{shapeType=Capsule, length=1, radius=1, shapeName=dyn_thumb_01_capsule, center=<2 0 0>, rotation=<rot 90 0 0>}]
 dyn_wrist_00 [{shapeType=Capsule, length=2, radius=1, shapeName=dyn_wrist_00_capsule, rotation=<rot 90 0 0>}]
}
constraints [
 {bone1Name=dyn_arm_00, bone2Name=Melee_TwoHanded_Handle, swing1LimitDegrees=5, swing2LimitDegrees=5, twistLimitDegrees=5}
 {bone1Name=dyn_wrist_00, bone2Name=dyn_arm_00, swing1LimitDegrees=25, swing2LimitDegrees=25, twistLimitDegrees=1}
 {bone1Name=dyn_hand_00, bone2Name=dyn_wrist_00, swing1LimitDegrees=1, swing2LimitDegrees=1, twistLimitDegrees=1}
 {bone1Name=dyn_index_00, bone2Name=dyn_hand_00, swing1LimitDegrees=20, swing2LimitDegrees=2, twistLimitDegrees=1}
 {bone1Name=dyn_index_02, bone2Name=dyn_index_00, swing1LimitDegrees=45, swing2LimitDegrees=2, twistLimitDegrees=1}
 {bone1Name=dyn_middle_00, bone2Name=dyn_hand_00, swing1LimitDegrees=20, swing2LimitDegrees=2, twistLimitDegrees=1}
 {bone1Name=dyn_middle_02, bone2Name=dyn_middle_00, swing1LimitDegrees=45, swing2LimitDegrees=2, twistLimitDegrees=1}
 {bone1Name=dyn_pinky_00, bone2Name=dyn_hand_00, swing1LimitDegrees=20, swing2LimitDegrees=2, twistLimitDegrees=1}
 {bone1Name=dyn_pinky_02, bone2Name=dyn_pinky_00, swing1LimitDegrees=45, swing2LimitDegrees=2, twistLimitDegrees=1}
 {bone1Name=dyn_ring_00, bone2Name=dyn_hand_00, swing1LimitDegrees=20, swing2LimitDegrees=2, twistLimitDegrees=1}
 {bone1Name=dyn_ring_02, bone2Name=dyn_ring_00, swing1LimitDegrees=45, swing2LimitDegrees=2, twistLimitDegrees=1}
 {bone1Name=dyn_thumb_00, bone2Name=dyn_hand_00, swing1LimitDegrees=2, swing2LimitDegrees=10, twistLimitDegrees=1}
 {bone1Name=dyn_thumb_01, bone2Name=dyn_thumb_00, swing1LimitDegrees=2, swing2LimitDegrees=35, twistLimitDegrees=1}
 {bone1Name=dyn_bodyWireA_00, bone2Name=Melee_TwoHanded_Handle, swing1LimitDegrees=20, swing2LimitDegrees=20, twistLimitDegrees=1}
 {bone1Name=dyn_bodyWireA_01, bone2Name=dyn_bodyWireA_00, swing1LimitDegrees=20, swing2LimitDegrees=20, twistLimitDegrees=1}
 {bone1Name=dyn_bodyWireA_02, bone2Name=dyn_bodyWireA_01, swing1LimitDegrees=20, swing2LimitDegrees=17, twistLimitDegrees=1}
 {bone1Name=dyn_bodyWireB_00, bone2Name=Melee_TwoHanded_Handle, swing1LimitDegrees=20, swing2LimitDegrees=20, twistLimitDegrees=1}
 {bone1Name=dyn_bodyWireB_01, bone2Name=dyn_bodyWireB_00, swing1LimitDegrees=20, swing2LimitDegrees=20, twistLimitDegrees=1}
 {bone1Name=dyn_bodyWireB_02, bone2Name=dyn_bodyWireB_01, swing1LimitDegrees=20, swing2LimitDegrees=20, twistLimitDegrees=1}
 {bone1Name=dyn_neck, bone2Name=Melee_TwoHanded_Handle, swing1LimitDegrees=7, swing2LimitDegrees=1, twistLimitDegrees=2}
 {bone1Name=dyn_head, bone2Name=dyn_neck, swing1LimitDegrees=7, swing2LimitDegrees=2, twistLimitDegrees=5}
 {bone1Name=dyn_hairA_00, bone2Name=dyn_head, swing1LimitDegrees=20, swing2LimitDegrees=20, twistLimitDegrees=1}
 {bone1Name=dyn_hairB_00, bone2Name=dyn_head, swing1LimitDegrees=20, swing2LimitDegrees=20, twistLimitDegrees=1}
 {bone1Name=dyn_hairC_00, bone2Name=dyn_head, swing1LimitDegrees=20, swing2LimitDegrees=20, twistLimitDegrees=1}
 {bone1Name=dyn_hairD_00, bone2Name=dyn_head, swing1LimitDegrees=20, swing2LimitDegrees=20, twistLimitDegrees=1}
 {bone1Name=dyn_hairE_00, bone2Name=dyn_head, swing1LimitDegrees=20, swing2LimitDegrees=20, twistLimitDegrees=1}
 {bone1Name=dyn_headWireA_00, bone2Name=dyn_head, swing1LimitDegrees=20, swing2LimitDegrees=20, twistLimitDegrees=1}
 {bone1Name=dyn_headWireB_00, bone2Name=dyn_head, swing1LimitDegrees=20, swing2LimitDegrees=20, twistLimitDegrees=1}
 {bone1Name=dyn_headWireC_00, bone2Name=dyn_head, swing1LimitDegrees=20, swing2LimitDegrees=20, twistLimitDegrees=1}
 {bone1Name=dyn_headWireD_00, bone2Name=dyn_head, swing1LimitDegrees=20, swing2LimitDegrees=20, twistLimitDegrees=1}
]
